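# Translucent two-sided particle material for the RainbowGlitch skydiving contrail, drawing a soft square with a procedural rainbow gradient.
# Offset UV ramps build the RGB bands, with switches for orientation and flipping, then get particle color, time-of-day scaling and desaturation for emissive.
# Opacity multiplies per-axis SphereMasks, hardened by dynamic parameters, with particle alpha, view-angle fade and optional edge, detail and camera fades.
# Depth and near-camera fading are skipped on mobile.
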Material /BRCosmetics/Athena/Effects/Materials/Skydiving/P_Contrails_RainbowGlitch/M_SquareRainbow
graph {
 properties {
  bUsedWithBeamTrails=true
  bUsedWithNiagaraMeshParticles=true
  bUsedWithNiagaraRibbons=true
  bUsedWithNiagaraSprites=true
  bUsedWithParticleSprites=true
  blendMode=BLEND_Translucent
  pixelDepthOffsetMode=PDOM_Legacy
  twoSided=true
  usedShadingModels=Default Lit
 }
 nodes [
  n1 = TextureCoordinate() {uTiling=2, vTiling=2}
  n2 = StaticSwitchParameter(True=ComponentMask(None=@n1) {g=true}, False=ComponentMask(None=@n1) {r=true}) {parameterName=HorizontalBars}
  n3 = OneMinus(None=Abs(None=Add(A=@n2) {constB=-2}))
  n4 = OneMinus(None=Abs(None=Add(A=@n2) {constB=-1}))
  n5 = Multiply(A=@n4, B=@n4)
  n6 = OneMinus(None=Abs(None=@n2))
  n7 = MaterialFunctionCall(X=@n3, Y=@n5, Z=@n6) {materialFunction=&MakeFloat3}
  n8 = MaterialFunctionCall(X=@n6, Y=@n5, Z=@n3) {materialFunction=&MakeFloat3}
  n9 = ParticleColor()
  n10 = Multiply(A=@n9.RGB) {constB=1.2}
  n11 = Multiply(A=@n9.RGB) {constB=1.5}
  n12 = Desaturation(
   None=Multiply(
    A=StaticSwitchParameter(True=@n7.Result, False=@n8.Result) {parameterName=FlipColorRange}
    B=MaterialFunctionCall(Color_Morning=@n10, Color_Day=@n11, Color_Evening=@n10, Color_Night=@n9.RGB) {materialFunction=&Color_Time_of_Day}.Result
   )
   Fraction=ScalarParameter() {parameterName=Desaturation}
  )
  n13 = ParticleColor()
  MaterialFunctionCall(Emissive=LinearInterpolate(A=@n12, B=Multiply(A=@n13.RGB, B=@n12), Alpha=ScalarParameter() {parameterName=UseParticleColor})) {materialFunction=&MF_FX_EmissiveAdjustments}
  n14 = StaticSwitchParameter(True=@n4, False=Constant() {r=1}) {parameterName=NoHardEdges?}
  n15 = TextureCoordinate()
  n16 = ComponentMask(None=@n15) {r=true}
  n17 = Constant() {r=0.5}
  n18 = DynamicParameter() {defaultValue=<rgba 0.5 0.5 1 1>, paramNames [Hardness X, Hardness Y, Param3, Param4]}
  n19 = ComponentMask(None=@n15) {g=true}
  n20 = DepthFade(
   Opacity=MaterialFunctionCall() {materialFunction=&Near_Camera_Fade}.Result
   FadeDistance=ScalarParameter() {defaultValue=50, parameterName=DepthFade}
  )
  n21 = Constant() {r=1}
  n22 = Multiply(
   A=Multiply(
    A=SphereMask(A=@n16, B=@n17, Radius=@n17, Hardness=@n18.Hardness X) {hardnessPercent=50}
    B=SphereMask(A=@n19, B=@n17, Radius=@n17, Hardness=@n18.Hardness Y) {hardnessPercent=90}
   )
   B=Multiply(A=@n9.A, B=ShadingPathSwitch(Default=@n20, Deferred=@n20, Mobile=@n21) {desc=Migrated from FeatureLevelSwitch})
  )
  n23 = ScalarParameter() {parameterName=FadeBasedOnViewAngle}
  n24 = ComponentMask(None=TextureCoordinate()) {r=true}
  n25 = Power(Base=Multiply(A=OneMinus(None=@n24)) {constB=2.5}) {constExponent=4}
  n26 = TextureCoordinate()
  n27 = Power(Base=Multiply(A=ComponentMask(None=@n26) {r=true}) {constB=3})
  n28 = LinearInterpolate(A=Constant() {r=1}, B=Multiply(A=Saturate(None=@n25), B=Saturate(None=@n27)), Alpha=ScalarParameter() {parameterName=Use_Detail_Mask})
  n29 = MaterialFunctionCall(Distance=ScalarParameter() {parameterName=CameraFadeDistance}) {materialFunction=&Near_Camera_Fade}
  Multiply(
   A=Multiply(
    A=Multiply(A=@n14, B=@n22)
    B=MaterialFunctionCall("FadeSensitivity (optional)"=@n23) {materialFunction=&FadeBasedOnViewAngle}.Result
   )
   B=LinearInterpolate(A=@n28, B=Multiply(A=@n28, B=@n29.Result), Alpha=ScalarParameter() {parameterName=Use_Camera_Fade})
  )
 ]
}
